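# Additive, unlit, two-sided plasma material for Niagara mesh particles in the ShockAndSlow Launcher effects.
# Two scales of panning plasma coils, UV-distorted by soft noise, are lighten-blended and masked by distorted scrolling noise.
# Opacity combines sine-pulsed contrast, Fresnel edge subtraction and view-angle fade, with Erode dynamic-parameter dissolve, depth fade and optional MaskCenter Fresnel masking.
# Emissive sums HDR-boosted plasma and a warm-tinted noise core, multiplied by particle color.
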Material /Game/Effects/Fort_Effects/GameplayPulginsAssets/ShockAndSlow_Launcher/M_DeckBed_FX_Plasma
graph {
 properties {
  bUsedWithNiagaraMeshParticles=true
  blendMode=BLEND_Additive
  pixelDepthOffsetMode=PDOM_Legacy
  shadingModel=MSM_Unlit
  twoSided=true
  usedShadingModels=Unlit
 }
 nodes [
  n1 = ParticleColor()
  n2 = TextureCoordinate()
  n3 = ScalarParameter() {defaultValue=3, parameterName=Tile}
  n4 = AppendVector(A=ScalarParameter() {defaultValue=1, parameterName=Tile_U}, B=Divide(A=@n3))
  n5 = Multiply(A=TextureCoordinate(), B=@n3)
  n6 = ComponentMask(None=TextureSample(UVs=Panner(Coordinate=@n5) {speedX=0.5, speedY=0.5}) {texture=/Game/Effects/Fort_Effects/Emotes/UltraLight/Texture/T_SoftNoise_BC7}.RGB) {g=true, r=true}
  n7 = ScalarParameter() {defaultValue=6, parameterName=UV_Distortion}
  n8 = Panner(Coordinate=Add(A=Multiply(A=@n2, B=@n4), B=Divide(A=@n6, B=@n7))) {speedX=-1, speedY=-0.5}
  n9 = Multiply(A=@n8) {constB=2}
  n10 = Multiply(A=TextureCoordinate(), B=@n4)
  n11 = TextureCoordinate()
  n12 = TextureSample(UVs=Panner(Coordinate=Multiply(A=@n11, B=@n3)) {speedX=-0.5, speedY=0.5}) {samplerType=SAMPLERTYPE_LinearGrayscale, texture=/Game/Environments/Hexmap/HexClouds/T_HexmapFog_01_Height02}
  n13 = Divide(A=ComponentMask(None=@n12.RGB) {g=true, r=true}, B=@n7)
  n14 = TextureSample(UVs=Panner(Coordinate=Add(A=@n10, B=@n13)) {speedX=1, speedY=1}) {texture=/Game/Effects/Fort_Effects/Textures/Tile/T_Noise_02}
  n15 = Constant() {r=5}
  n16 = Multiply(
   A=MaterialFunctionCall!(
    Blend=TextureSample(UVs=@n8) {samplerType=SAMPLERTYPE_LinearGrayscale, texture=/Game/Effects/Fort_Effects/Emotes/UltraLight/Texture/T_PlasmaCoils}.R
    Base=TextureSample(UVs=@n9) {samplerType=SAMPLERTYPE_LinearGrayscale, texture=&T_PlasmaCoils}.R
   ) {materialFunction=/Engine/Functions/Engine_MaterialFunctions03/Blends/Blend_Lighten}.Result
   B=Saturate(None=Multiply(A=@n14.RGB, B=@n15))
  )
  n17 = Constant() {r=0.5}
  n18 = Multiply(A=@n17) {constB=2}
  n19 = MaterialFunctionCall(Speed=Constant() {r=2}) {materialFunction=&DebugTimeSine}
  n20 = Subtract(
   A=Power(Base=@n16, Exp=LinearInterpolate(A=@n17, B=@n18, Alpha=@n19.Result)) {constExponent=1.5}
   B=Abs(None=Fresnel() {baseReflectFraction=0, exponent=3})
  )
  n21 = MaterialFunctionCall("FadeSensitivity (optional)"=Multiply(A=LinearInterpolate(Alpha=@n12.R) {constA=0.1}, B=Constant() {r=50})) {materialFunction=&FadeBasedOnViewAngle}
  n22 = Subtract(
   A=Add(A=Saturate(None=@n20), B=@n21.Result)
   B=DynamicParameter() {defaultValue=<rgba 0 1 1 1>, paramNames [Erode,Param2,Param3,Param4]}.Erode
  )
  n23 = DepthFade(Opacity=Multiply(A=@n1.A, B=Saturate(None=@n22)))
  LinearInterpolate(A=@n23, B=Multiply(A=@n23, B=Fresnel() {baseReflectFraction=0, exponent=1}), Alpha=ScalarParameter() {parameterName=MaskCenter})
  n24 = Constant() {r=100}
  n25 = MaterialFunctionCall(LDR Input=@n16, HDR Dynamic Range=@n24, HDR Tint=Multiply(A=ScalarParameter() {defaultValue=100, parameterName=NoiseBrightness}, B=@n16)) {materialFunction=&DeriveHDRfromLDR}
  n26 = TextureSample() {texture=&T_SoftNoise_BC7}
  n27 = Constant() {r=0.6}
  n28 = Multiply(A=Multiply(A=Desaturation(None=@n26.RGB, Fraction=@n27), B=VectorParameter() {defaultValue=<rgba 2 1.3923 0.6267 1>}.RGB), B=@n21.Result)
  n29 = MaterialFunctionCall(
   LDR Input=@n28
   HDR Dynamic Range=@n24
   HDR Intensity Multiplier=Multiply(A=@n28, B=ScalarParameter() {defaultValue=20, parameterName=CoreBrightness})
  ) {materialFunction=&DeriveHDRfromLDR}
  MaterialFunctionCall(Emissive=Multiply(A=Add(A=@n25.RemappedRange, B=@n29.RemappedRange), B=@n1.RGB)) {materialFunction=&MF_FX_EmissiveAdjustments}
 ]
}
