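# Translucent, unlit, two-sided Niagara mesh-particle material for the Fortanium shockwave in the Shock and Slow launcher effects.
# Dynamic parameters and a UV gradient warp and offset panning magnetic-wave and caustic textures, combined with Max.
# Inner/outer tints, particle color and HDR remapping shape emissive; particle alpha, edge mask, view-angle fade, optional vertex color and quality-switched depth fade shape opacity.
# A pinch term scales vectors toward particle center for vertex offset.
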Material /Game/Effects/Fort_Effects/GameplayPulginsAssets/ShockAndSlow_Launcher/M_Fortanium_Shockwave_Mesh
graph {
 properties {
  bUsedWithNiagaraMeshParticles=true
  blendMode=BLEND_Translucent
  pixelDepthOffsetMode=PDOM_Legacy
  shadingModel=MSM_Unlit
  twoSided=true
  usedShadingModels=Unlit
 }
 nodes [
  n1 = MaterialFunctionCall(Float2=TextureCoordinate()) {materialFunction=&BreakOutFloat2Components}
  n2 = OneMinus(None=@n1.G)
  n3 = Power(Base=@n1.G, Exp=ScalarParameter() {defaultValue=2, parameterName=GradientPower}) {constExponent=3}
  n4 = ScalarParameter() {defaultValue=0.4, parameterName=BottomWarp}
  n5 = DynamicParameter() {defaultValue=<rgba 1 1 1 0>, paramNames [WPOPinch,Warp,PanAdd,Blend]}
  n6 = Add(A=Multiply(A=NamedRerouteDeclaration(None=@n3) {name=Gradient, nodeColor=<rgba 0 1 0.5781 1>}, B=Multiply(A=@n4, B=@n5.Warp)), B=@n5.PanAdd)
  n7 = Constant2Vector() {g=1}
  n8 = VectorParameter() {parameterName=DistortionTile, defaultValue=<rgba 1 0.8 0 0>}
  n9 = Multiply(A=Add(A=AppendVector(A=@n1.R, B=@n2), B=Multiply(A=@n6, B=@n7)), B=AppendVector(A=@n8.R, B=@n8.R))
  n10 = MaterialFunctionCall(XY=Panner(Coordinate=@n9) {speedX=0.1, speedY=-0.2}) {materialFunction=&Swizzle}
  n11 = Constant2Vector() {g=1, r=1.5}
  n12 = LinearInterpolate(
   A=TextureSample(UVs=@n10.YX) {texture=/Game/Athena/Environments/Nexus/Textures/T_ZeroPoint_MagneticWaves}.RGB
   B=TextureSample(UVs=Multiply(A=@n10.YX, B=@n11)) {texture=/Game/Athena/Environments/Nexus/Textures/T_ZeroPoint_MagneticWaves_Blurred}.RGB
   Alpha=NamedRerouteUsage()
  )
  n13 = TextureSample(UVs=Multiply(A=Panner(Coordinate=@n9) {speedX=0.1, speedY=-0.6}, B=Constant2Vector() {g=1, r=1})) {texture=/Game/Effects/Fort_Effects/Textures/Tile/T_ShieldCaustics_Tile_Chromatic}
  LinearInterpolate(A=@n12, B=@n13.RGB) {constAlpha=0.7}
  n14 = ParticleColor()
  n15 = NamedRerouteUsage()
  n16 = Multiply(A=ScalarParameter() {defaultValue=0.75, parameterName=BackgroundBrightness}, B=ComponentMask(None=TextureCoordinate()) {g=true})
  n17 = VectorParameter() {parameterName=OuterColor, defaultValue=<rgba 0.1854 0.1828 0.5 0>}
  n18 = VectorParameter() {parameterName=InnerColor, defaultValue=<rgba 0.9244 0.379 1 0>}
  n19 = NamedRerouteUsage()
  n20 = Multiply(A=@n19, B=@n19)
  n21 = Multiply(A=@n14.RGB, B=Multiply(A=Max(A=@n15, B=@n16), B=LinearInterpolate(A=@n17.RGB, B=@n18.RGB, Alpha=@n20)))
  MaterialFunctionCall(Emissive=MaterialFunctionCall(LDR Input=Clamp(None=@n21), HDR Tint=@n21, HDR Intensity Multiplier=ScalarParameter() {defaultValue=5, parameterName=HDRInensity}) {materialFunction=&DeriveHDRfromLDR}.RemappedRange) {materialFunction=&MF_FX_EmissiveAdjustments}
  n22 = ScalarParameter() {defaultValue=1, parameterName=Pinch}
  n23 = Subtract(A=ParticlePositionWS(), B=WorldPosition().XYZ)
  Multiply(A=Multiply(A=Multiply(A=@n22, B=@n23), B=NamedRerouteUsage()), B=NamedRerouteUsage())
  n24 = VertexColor()
  n25 = ScalarParameter() {parameterName=UseVertColorOpacity}
  n26 = ComponentMask(None=TextureCoordinate()) {g=true}
  n27 = Constant() {r=0.5}
  n28 = SphereMask(A=Power(Base=OneMinus(None=@n26)), B=@n27, Radius=@n27) {hardnessPercent=0}
  n29 = QualitySwitch(Default=DepthFade(Opacity=@n28, FadeDistance=ScalarParameter() {defaultValue=50, parameterName=DepthFade}), Low=@n28)
  n30 = MaterialFunctionCall("FadeSensitivity (optional)"=ScalarParameter() {defaultValue=1, parameterName=AngleFade}) {materialFunction=&FadeBasedOnViewAngle}
  n31 = Desaturation(None=NamedRerouteDeclaration(None=Max(A=@n12, B=@n13.RGB)) {name=EmissiveTexture, nodeColor=<rgba 0.4531 1 0 1>}, Fraction=Constant() {r=1})
  Multiply(
   A=Multiply(A=LinearInterpolate(B=@n24.G, Alpha=@n25) {constA=1}, B=Multiply(A=@n14.A, B=@n29))
   B=Multiply(A=Multiply(A=@n30.Result, B=@n30.Result), B=@n31)
  )
  NamedRerouteDeclaration(None=@n5.WPOPinch) {name=WPOPinchParam, nodeColor=<rgba 0.5938 0 1 1>}
  NamedRerouteDeclaration(None=@n5.Blend) {name=BlendParam, nodeColor=<rgba 1 0.4688 0 1>}
  Power(Base=@n31) {constExponent=0.6}
 ]
}
