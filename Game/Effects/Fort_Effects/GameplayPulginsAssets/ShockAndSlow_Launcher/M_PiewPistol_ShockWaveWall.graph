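# Translucent, two-sided unlit material for a shockwave wall effect on Niagara mesh particles, apparently for the PiewPistol Shock and Slow launcher.
# A noise-varied vertical glow gradient adds to world-XY-projected, panning plasma clouds, with a fade near V=1.
# Particle color tints this into emissive through MF_FX_EmissiveAdjustments, while particle alpha and inverted Fresnel shape opacity.
# A quality switch drops depth fade on Low and Medium; one plasma sample is unconnected.
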Material /Game/Effects/Fort_Effects/GameplayPulginsAssets/ShockAndSlow_Launcher/M_PiewPistol_ShockWaveWall
graph {
 properties {
  bUsedWithNiagaraMeshParticles=true
  blendMode=BLEND_Translucent
  pixelDepthOffsetMode=PDOM_Legacy
  shadingModel=MSM_Unlit
  twoSided=true
  usedShadingModels=Unlit
 }
 nodes [
  n1 = ParticleColor()
  n2 = Panner(Coordinate=Multiply(A=TextureCoordinate(), B=Constant2Vector() {g=0.3, r=2})) {speedX=0.5, speedY=1}
  n3 = Power(
   Base=ComponentMask(None=TextureCoordinate()) {g=true}
   Exp=LinearInterpolate(Alpha=TextureSample(UVs=@n2) {texture=/Game/Effects/FogVolume/noise_flamy_Tile}.R) {constA=3, constB=6}
  )
  n4 = ScalarParameter() {defaultValue=1, parameterName=Glow Intensity}
  n5 = Power(Base=ComponentMask(None=TextureCoordinate()) {g=true})
  n6 = Divide(A=ComponentMask(None=WorldPosition().XYZ) {g=true, r=true}, B=ScalarParameter() {defaultValue=50000, parameterName=Tiling_WorldUnits}) {constB=50000}
  n7 = Multiply(A=Time(), B=ScalarParameter() {defaultValue=0.1, parameterName=Panner Speed})
  n8 = Multiply(
   A=Sine(None=Multiply(A=@n5) {constB=0.5})
   B=TextureSample(UVs=Panner(Coordinate=@n6, Time=@n7) {speedX=1}) {texture=/Game/Weapons/WeaponSkins/Wraps/Textures/PlasmaClouds/T_Wrap_PlasmaClouds__M}.G
  )
  n9 = ScalarParameter() {defaultValue=1, parameterName=Fire Intensity}
  n10 = OneMinus(None=ComponentMask(None=TextureCoordinate()) {g=true})
  n11 = Multiply(A=Add(A=Multiply(A=@n3, B=@n4) {constB=0.35}, B=Multiply(A=@n8, B=@n9)), B=Saturate(None=Multiply(A=@n10) {constB=7}))
  MaterialFunctionCall(Emissive=Multiply(A=@n1.RGB, B=@n11)) {materialFunction=&MF_FX_EmissiveAdjustments}
  n12 = ScalarParameter() {defaultValue=3, parameterName=Fresnel Exponent}
  n13 = Multiply(A=Multiply(A=@n1.A, B=@n11), B=OneMinus(None=Fresnel(ExponentIn=@n12) {baseReflectFraction=0, exponent=3}))
  n14 = Multiply(A=@n13, B=DepthFade(FadeDistance=ScalarParameter() {defaultValue=1000, parameterName=DepthFade}) {fadeDistanceDefault=1000})
  QualitySwitch(Default=@n14, Low=@n13, High=@n14, Medium=@n13, Epic=@n14)
  n15 = Time()
  n16 = ScalarParameter() {defaultValue=0.1, parameterName=Panner Speed}
  TextureSample(UVs=Panner(Coordinate=TextureCoordinate() {uTiling=4, vTiling=0.1}, Time=Multiply(A=@n15, B=@n16)) {speedY=1}) {texture=&T_Wrap_PlasmaClouds__M}
 ]
}
