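# Translucent, unlit, two-sided tech-line effect material for Niagara mesh particles, filed with the ShockAndSlow launcher effects.
# Two differently scaled T_TechLines samples (G plus boosted R) form a mask; the G layer's panner uses constant time, so its offset is static.
# The mask indexes a Storm King curve atlas row for color, which is hue-shifted, tinted by particle color and passed through emissive adjustments.
# The same mask multiplied by particle alpha supplies transparency.
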Material /Game/Effects/Fort_Effects/GameplayPulginsAssets/ShockAndSlow_Launcher/M_Tech_Triangles_DualRev
graph {
 properties {
  bUsedWithNiagaraMeshParticles=true
  blendMode=BLEND_Translucent
  pixelDepthOffsetMode=PDOM_Legacy
  shadingModel=MSM_Unlit
  twoSided=true
  usedShadingModels=Unlit
 }
 nodes [
  n1 = ParticleColor()
  n2 = ScalarParameter() {defaultValue=0.1, parameterName=Hueshift}
  n3 = TextureCoordinate()
  n4 = Constant2Vector() {g=0.1}
  n5 = Panner(
   Coordinate=MaterialFunctionCall(UVs=Add(A=@n3, B=@n4), Texture Scale=Constant2Vector() {g=4, r=2.7}) {materialFunction=&ScaleUVsByCenter}.UVs
   Time=Constant() {r=0.1}
  ) {speedX=0.2}
  n6 = TextureSample(UVs=TextureCoordinate() {uTiling=1.5, vTiling=20}) {texture=/Game/Effects/Fort_Effects/Textures/T_TechLines}
  n7 = Multiply(A=Add(A=TextureSample(UVs=@n5) {texture=&T_TechLines}.G, B=Multiply(A=@n6.R) {constB=5}))
  n8 = CurveAtlasRowParameter(CurveTime=@n7) {atlas=&AthenaCurveAtlas01, curve=/Game/Characters/Enemies/DUDEBRO/FX/Curve_StormKing_Alpha, defaultValue=141}
  MaterialFunctionCall(Emissive=Multiply(
    A=@n1.RGB
    B=MaterialFunctionCall(Hue Shift Percentage=@n2, Texture=@n8) {materialFunction=&HueShift}.Result
   )) {materialFunction=&MF_FX_EmissiveAdjustments}
  Multiply(A=@n1.A, B=@n7)
 ]
}
