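# Unlit, two-sided translucent trail material for Niagara ribbons, named for dual revolvers in the ShockAndSlow launcher content.
# An inverted triangle glitch texture blends two green Derez colors, tinted by particle color for glow.
# Parabolic masks from two UV sets soften ribbon edges, combining with particle alpha, glitch-driven boost and DepthFade for opacity.
# A time-animated SmoothStep branch appears unconnected.
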Material /Game/Effects/Fort_Effects/GameplayPulginsAssets/ShockAndSlow_Launcher/M_Trail_DualRevolvers
graph {
 properties {
  bAutomaticallySetUsageInEditor=false
  bUsedWithNiagaraRibbons=true
  blendMode=BLEND_Translucent
  shadingModel=MSM_Unlit
  twoSided=true
  usedShadingModels=Unlit
 }
 nodes [
  n1 = TextureSample(UVs=TextureCoordinate()) {samplerSource=SSM_Wrap_WorldGroupSettings, texture=&/T_Triangles_Glitch}
  n2 = OneMinus(None=@n1.R)
  n3 = Multiply(A=Multiply(A=NamedRerouteUsage()) {constB=5}, B=@n2)
  n4 = ParticleColor()
  n5 = MaterialFunctionCall(Float2=TextureCoordinate() {coordinateIndex=1}) {materialFunction=&BreakOutFloat2Components}
  n6 = Multiply(A=@n5.R, B=OneMinus(None=@n5.R))
  n7 = MaterialFunctionCall(Float2=TextureCoordinate()) {materialFunction=&BreakOutFloat2Components}
  n8 = Multiply(A=Multiply(A=@n7.G, B=OneMinus(None=@n7.G))) {constB=2}
  n9 = Multiply(A=@n4.A, B=Multiply(A=Multiply(A=@n6) {constB=2}, B=Saturate(None=@n8)))
  DepthFade(Opacity=Add(A=Multiply(A=@n3, B=@n9), B=@n9), FadeDistance=ScalarParameter() {defaultValue=100, parameterName=DepthFade})
  n10 = LinearInterpolate(
   A=VectorParameter() {group=Derez, parameterName=Color_01, defaultValue=<rgba 0.054 0.2 0 1>}.RGB
   B=VectorParameter() {group=Derez, parameterName=Color_02, defaultValue=<rgba 0 5 0.1475 1>}.RGB
   Alpha=Multiply(A=@n2, B=Power(Base=@n2) {constExponent=0.5})
  )
  n11 = LinearInterpolate(A=@n10, B=Multiply(A=@n10) {constB=5}) {constAlpha=0}
  Multiply(A=@n11, B=@n4.RGB)
  NamedRerouteDeclaration(None=@n11) {name=Emissive, nodeColor=<rgba 0 0.3438 1 1>}
  n12 = Time()
  SmoothStep(Max=Abs(None=Sine(None=@n12)), Value=@n1.G)
 ]
}
